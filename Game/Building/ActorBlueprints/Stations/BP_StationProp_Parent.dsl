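; Parent Blueprint for interactable station props, such as vending machines, supplying shared interaction, marker and ambient audio behavior.
; On authority, it binds interaction to send a vending-machine-used gameplay event to the pawn, regardless of purchase; remotes optionally start ambient audio, which fades out on death.
; Interaction requires it to be allowed, the prop intact and the player's controller outside any active conversation.
; Variables supply prompt, timing and marker data; PlayServiceUsedEffects is empty.
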(event Custom|MachineInteractedWith
  (params (SelfActor :type Building%20Actor%20Object%20Reference) (InteractingPawn :type Fort%20Pawn%20Object%20Reference))
  ;[1 1 1] Generic Event so we can track when players interact with a Vending Machine (We don't need to know if they purchased something or not.
  (Ability|SendGameplayEventtoActor InteractingPawn "(TagName=\"Event.VendingMachines.Used\")"))

(event EventBeginPlay
  (|CallParentFunction)
  (Networking|SwitchHasAuthority
    (:Authority
      ;[1 1 1] Generic Event so we can track when players interact with a Vending Machine (We don't need to know if they purchased something or not.
      (Interaction|BindEventtoOnInteract self (delegate Custom|MachineInteractedWith)))
    (:Remote
      (if (Variables|Default|GetStartAmbientOnBeginPlay)
        (Utilities|IsValid (Variables|/Script/Engine.AudioComponent|GetSound (Variables|Components|GetStation_Ambient))
          (:Is%20Valid
            (Audio|Components|Audio|Play (Variables|Components|GetStation_Ambient) 0.0)))))))

(event Collision|EventActorBeginOverlap
  (params (OtherActor :type Actor%20Object%20Reference)))

(event EventTick
  (params (DeltaSeconds :type Float%20%28single-precision%29)))

(event EventOnDeathPlayEffects
  (params (Damage :type Float%20%28single-precision%29) (DamageTags :type Gameplay%20Tag%20Container%20Structure%20%28by%20ref%29) (Momentum :type Vector :default "0, 0, 0") (HitInfo :type Hit%20Result%20Structure%20%28by%20ref%29) (InstigatedBy :type Fort%20Pawn%20Object%20Reference) (DamageCauser :type Actor%20Object%20Reference) (EffectContext :type Gameplay%20Effect%20Context%20Handle%20Structure))
  (Audio|Components|Audio|FadeOut (Variables|Components|GetStation_Ambient) 0.1 0.0 "Linear" :!class /Script/Engine.AudioComponent))

(fn BlueprintCanInteract
  (params (InteractingPawn :type Fort%20Pawn%20Object%20Reference) (InteractionBeingAttempted :type EInteractionBeingAttempted%20Enum :default "FirstInteraction") (InteractionType :type TInteractionType%20Enum :default "IT_NoInteraction"))
  (if (and (Variables|Interaction|GetbAllowInteract) (not (Variables|BuildingActor|GetbDestroyed)))
    (bind _returnvalue (Conversation|FindConversationComponent (Pawn|GetController InteractingPawn)))
    (Utilities|IsValid _returnvalue
      (:Is%20Valid
        (return (not (Conversation|IsinActiveConversation _returnvalue))))
      (:Is%20Not%20Valid
        (return true)))
    (else
      (return false))))

(fn BlueprintGetInteractionString
  (params (InteractingPawn :type Fort%20Pawn%20Object%20Reference) (InteractionBeingAttempted :type EInteractionBeingAttempted%20Enum :default "FirstInteraction"))
  (return (Variables|GetBuildingActorInteractionText)))

(fn BlueprintGetInteractionTime
  (params (InteractingPawn :type Fort%20Pawn%20Object%20Reference) (InteractionBeingAttempted :type EInteractionBeingAttempted%20Enum :default "FirstInteraction"))
  (return (Variables|Interaction|GetInteractionSpeedOverride) true))

(fn GetMarkerDisplayData
  (return (Variables|Marking|GetMarkerDisplay)))

(fn GetMarkerPositionOffset
  (return (Variables|Marking|GetMarkerPositionOffset)))

(fn PlayServiceUsedEffects
  (params (PlayerController :type Fort%20Player%20Controller%20Object%20Reference)))

(fn ConstructionScript
  (|CallParentFunction))
