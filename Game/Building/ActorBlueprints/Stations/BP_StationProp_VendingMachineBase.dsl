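; Base vending machine station prop that plays cosmetic feedback when its service is used.
; On non-authority instances, BeginPlay caches the mesh's relative rotation and, if material effects are enabled, creates a dynamic monitor material.
; PlayServiceUsedEffects triggers a sound, loot particle and timeline that flashes the monitor and wobbles the mesh around its cached rotation.
; The wobble applies only while a valid physics object component is asleep.
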(event EventBeginPlay
  (|CallParentFunction)
  (Networking|SwitchHasAuthority
    (:Remote
      (Variables|Default|SetInitialRelativeRotation (Variables|/Script/Engine.SceneComponent|GetRelativeRotation (Variables|Debug|GetStaticMeshComponent)))
      (if (Variables|Default|GetMaterialEffectsEnabled)
        (bind _returnvalue (Rendering|Material|CreateDynamicMaterialInstance (Variables|Debug|GetStaticMeshComponent) (Rendering|Material|GetMaterialIndex (Variables|Debug|GetStaticMeshComponent) "M_VendingMachineMonitor") (Rendering|Material|GetMaterial (Variables|Debug|GetStaticMeshComponent) (Rendering|Material|GetMaterialIndex (Variables|Debug|GetStaticMeshComponent) "M_VendingMachineMonitor")) "None" :!class /Script/Engine.PrimitiveComponent))
        (Variables|VendingMachineMonitorSettings|SetMonitor%20Mat _returnvalue)))))

(event Collision|EventActorBeginOverlap
  (params (OtherActor :type Actor%20Object%20Reference))
  (|CallParentFunction OtherActor) ; This node is disabled and will not be called.\nDrag off pins to build functionality.
)

(event EventTick
  (params (DeltaSeconds :type Float%20%28single-precision%29))
  (|CallParentFunction DeltaSeconds) ; This node is disabled and will not be called.\nDrag off pins to build functionality.
)

(event Custom|TriggerVendEffects
  (bind _p_loot_effect_vending (Variables|Default|GetP_Loot_Effect_Vending))
  (Utilities|FlowControl|Sequence
    (:then_0
      (Audio|PlaySoundatLocation (Variables|StationSounds|GetVend%20Succeeded%20Sound) (Transformation|GetActorLocation))
      (bind (_direction _wobbleovertime _flash) (Timeline|WobbleTimeline PlayFromStart)
        (:Update
          (Utilities|FlowControl|Sequence
            (:then_0
              (if (Variables|Default|GetMaterialEffectsEnabled)
                (Rendering|Material|SetScalarParameterValue (Variables|VendingMachineMonitorSettings|GetMonitor%20Mat) (Variables|VendingMachineMonitorSettings|GetMonitorMaterialParamName) _flash :!class /Script/Engine.MaterialInstanceDynamic)))
            (:then_1
              (bind _returnvalue (BuildingActor|PhysicsSimulationSupport|GetPhysicsObjectComponent self))
              (Utilities|IsValid _returnvalue
                (:Is%20Valid
                  (if (not (PhysicsObject|IsAwake _returnvalue))
                    (Transformation|SetRelativeRotation (Variables|Debug|GetStaticMeshComponent) (Math|Rotator|CombineRotators (Variables|Default|GetInitialRelativeRotation) (Math|Rotator|Lerp(Rotator) "0, 0, 0" "0.500000,0.000000,0.000000" _wobbleovertime true)) false false)))))))
        (:Finished
          (if (Variables|Default|GetMaterialEffectsEnabled)
            (Rendering|Material|SetScalarParameterValue (Variables|VendingMachineMonitorSettings|GetMonitor%20Mat) (Variables|VendingMachineMonitorSettings|GetMonitorMaterialParamName) 0.0 :!class /Script/Engine.MaterialInstanceDynamic)))))
    (:then_1
      (Utilities|IsValid _p_loot_effect_vending
        (:Is%20Valid
          (Components|Activation|SetActive _p_loot_effect_vending true true))))))

(fn PlayServiceUsedEffects
  (params (PlayerController :type Fort%20Player%20Controller%20Object%20Reference))
  (CallFunction|TriggerVendEffects))

(fn ConstructionScript
  (|CallParentFunction))
