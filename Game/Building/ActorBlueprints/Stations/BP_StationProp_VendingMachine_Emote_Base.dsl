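; Base Blueprint for emote vending machines whose monitor face reacts to players through material styles, sounds and speech bubbles.
; Clients reveal the screen on approach and request styles from conversation metadata, sustained local dancing or health-tiered damage responses.
; Timelines blend emote textures, colors and distortion, and drive vend wobble.
; The authority destroys disallowed machines and watches for EMP tags, which trigger effects, timed pickup spawning and EMP removal.
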(event Custom|EMP_StartWaitEMPEvents
  ;[0.3137 0.5699 1] EMP
  ;;[1 1 1] Start Wait EMP Events
  (Ability|Tasks|WaitGameplayTagAddToActor 0 (Variables|EMP|Tags|GetTag_EMPApplied)
    (:then
      (Ability|Tasks|WaitGameplayTagRemoveFromActor 0 (Variables|EMP|Tags|GetTag_EMPApplied)
        (:Removed
          (CallFunction|EMPSetApplied false))))
    (:Added
      (CallFunction|EMPSetApplied true))))

(event Custom|RepEMPFX
  ;[0.3137 0.5699 1] EMP
  ;;[1 1 1] Rep EMP FX
  (CallFunction|PlayServiceUsedEffects 0)
  (Utilities|IsValid (Variables|EMP|FX|GetFX_EmpHit)
    (:Is%20Valid
      (Effects|Components|ParticleSystem|SpawnFXAtLocation (Variables|EMP|FX|GetFX_EmpHit) (Transformation|GetWorldLocation (Variables|Debug|GetStaticMeshComponent)) (Math|Rotator|MakeRotator (.roll (Math|Conversions|RotationFromXVector (Transformation|GetForwardVector (Variables|Debug|GetStaticMeshComponent)))) (.pitch (Math|Conversions|RotationFromXVector (Transformation|GetForwardVector (Variables|Debug|GetStaticMeshComponent)))) (+ (.yaw (Math|Conversions|RotationFromXVector (Transformation|GetForwardVector (Variables|Debug|GetStaticMeshComponent)))) 9e+01))))
    (:Is%20Not%20Valid
      _))
  (if (Utilities|Array|IsNotEmpty (Variables|EMP|FX|GetEMP_StyleChangeOptions))
    (Class|FortMaterialStyleComponent|RequestStyleChange (Variables|Default|GetFortMaterialStyleComponent) (Utilities|Struct|MakeRequestStyleParams -1.0 (Variables|Materials|GetMonitor%20Mat) "StyleSetLookup" (Variables|Materials|GetMaterialStyleSet) (Utilities|Array|RandomArrayItem (Variables|EMP|FX|GetEMP_StyleChangeOptions))) "ReplaceActive")))

(event Custom|EMP_SetApplied
  (params (EMPApplied :type Boolean))
  ;[0.3137 0.5699 1] EMP
  ;;[1 1 1] EMP Set Applied
  (if EMPApplied
    (if (ScalableFloat|GetValueasBool (Variables|EMP|Balance|GetRow_EMPEnabled))
      (CallFunction|RepEMPFX)
      (Utilities|Time|SetTimerbyEvent (EventDispatchers|CreateEvent :!event Custom|EMP_SpawnItem) (Variables|EMP|Timing|GetDelay_ItemSpawn))
      (Utilities|Time|SetTimerbyEvent (EventDispatchers|CreateEvent :!event Custom|EMP_RemoveEMP) (Math|Float|Max(Float) (Variables|EMP|Timing|GetDelay_RemoveEMP) (Variables|EMP|Timing|GetDelay_ItemSpawn)))
      (else
        (CallFunction|EMPRemoveEMP)))
    (else
      (CallFunction|EMPRemoveEMP))))

(event Custom|EMP_SpawnItem
  (bind (_origin _boxextent) (Collision|GetActorBounds true))
  (bind (_returnvalue_roll _returnvalue_pitch _returnvalue_yaw) (Math|Rotator|BreakRotator (Transformation|GetActorRotation)))
  ;[0.3137 0.5699 1] EMP
  ;;[1 1 1] EMP Spawn Item
  (CallFunction|EMPSetItemandSpawnCount)
  (if (Math|Boolean|ANDBoolean (> (Building|GetBuildingHealthPercentage self) 0) (Utilities|IsValid (Variables|EMP|Vars|GetEMP_ItemToSpawn) :!class /Script/Engine.KismetSystemLibrary) (> (Variables|EMP|Vars|GetEMP_ItemCountToSpawn) 0))
    (Inventory|SpawnPickupInWorld (Variables|EMP|Vars|GetEMP_ItemToSpawn) (Variables|EMP|Vars|GetEMP_ItemCountToSpawn) (Utilities|Operators|Add _origin (* _boxextent (Math|Rotator|GetRotationXVector (Math|Rotator|MakeRotator _returnvalue_roll _returnvalue_pitch (+ _returnvalue_yaw (Variables|EMP|Item|GetItemSpawnRotationFromActorRotation))))) (* (Math|Rotator|GetRotationXVector (Math|Rotator|MakeRotator _returnvalue_roll _returnvalue_pitch (+ _returnvalue_yaw (Variables|EMP|Item|GetItemSpawnRotationFromActorRotation)))) (Variables|EMP|Item|GetEMP_ItemSpawnDistance))) (* (Math|Rotator|GetRotationXVector (Math|Rotator|MakeRotator _returnvalue_roll _returnvalue_pitch (+ _returnvalue_yaw (Variables|EMP|Item|GetItemSpawnRotationFromActorRotation)))) (Variables|EMP|Item|GetEMP_ItemSpawnDistance)) 0 true false false -1 "Container" "VendingMachine" 0 false)))

(event Custom|EMP_RemoveEMP
  ;[0.3137 0.5699 1] EMP
  ;;[1 1 1] Remove EMP
  (GameplayEffects|RemoveActiveEffectswithGrantedTags (Ability|GetFortAbilitySystemComponent self) (GameplayTags|MakeGameplayTagContainerfromTag (Variables|EMP|Tags|GetTag_EMPApplied))))

(event Collision|EventActorBeginOverlap
  (params (OtherActor :type Actor%20Object%20Reference))
  (|CallParentFunction OtherActor) ; This node is disabled and will not be called.\nDrag off pins to build functionality.
)

(event EventBeginPlay
  (Networking|SwitchHasAuthority
    (:Authority
      ;[0.281 1 0.4092] Destroy the vending machine if in Creative Royale.
      (|CallParentFunction)
      (if (not (ScalableFloat|GetValueasBool (Variables|Default|GetAllowedInMode)))
        (Actor|DestroyActor)
        (else
          (CallFunction|EMPStartWaitEMPEvents))))
    (:Remote
      (Variables|Materials|SetMonitor%20Mat (Class|FortMaterialStyleComponent|GetDynamicMaterialInstance (Variables|Default|GetFortMaterialStyleComponent)))
      (|CallParentFunction)
      (Variables|State|SetInitialRelativeRotation (Variables|/Script/Engine.SceneComponent|GetRelativeRotation (Variables|Debug|GetStaticMeshComponent)))
      ;[1 1 1] Event handling - Check for metadata driven Emote Trigger when we receive a new conversation message from the local player
      (EventDispatchers|BindEventtoOnReceivedLocalClientConversationMessage (Variables|Components|GetNonPlayerConversationComponent) (delegate Custom|Event%20On%20Receive%20Conversation%20Message))
      ;[1 1 1] Event Handling - detect local player enter/leave interaction range for player centric reactions
      (EventDispatchers|BindEventtoOnDialogMarkerInteractionStateChanged (Variables|Components|GetNonPlayerConversationComponent) (delegate Custom|Event%20On%20Dialog%20Marker%20Interaction%20State%20Changed)))))

(event Custom|Event%20Player%20Dance%20For%20Required%20Duration
  ;[1 1 1] Event Handling - detect local player enter/leave interaction range for player centric reactions
  (CallFunction|TriggerDanceReaction))

(event Custom|Event%20On%20Dialog%20Marker%20Interaction%20State%20Changed
  (params (CurrentInteractionState :type EDialogMarkerInteractionState%20Enum :default "Conversation") (InteractingController :type Fort%20Player%20Controller%20Object%20Reference))
  (bind _triggeremoteonlocalplayerdance (Variables|Responses|GetTriggerEmoteOnLocalPlayerDance))
  (bind _returnvalue (ScalableFloat|GetValueasBool _triggeremoteonlocalplayerdance))
  (bind _returnvalue_2 (Utilities|Enum|Equal(Enum) CurrentInteractionState "InteractionRange"))
  (bind _returnvalue_3 (Utilities|Enum|Equal(Enum) CurrentInteractionState "Conversation"))
  (bind _returnvalue_4 (or _returnvalue_3 _returnvalue_2))
  ;[1 1 1] Event Handling - detect local player enter/leave interaction range for player centric reactions
  (if (Pawn|IsLocalPlayerController InteractingController)
    (Utilities|FlowControl|Sequence
      (:then_0
        (if (Utilities|Enum|Equal(Enum) CurrentInteractionState "Conversation")
          (Variables|State|SetLastDamageMessageTime 0.0)))
      (:then_1
        (if (Math|Boolean|NotEqual(Boolean) _returnvalue_4 (or (Utilities|Enum|Equal(Enum) (Variables|State|GetLastDialogMarkerInteractionState) "Conversation") (Utilities|Enum|Equal(Enum) (Variables|State|GetLastDialogMarkerInteractionState) "InteractionRange")))
          (Variables|State|SetLastDialogMarkerInteractionState CurrentInteractionState)
          (if _returnvalue_4
            (CallFunction|ChangeEmoteScreenRevealed true)
            (Variables|State|SetLastInteractingPlayerController InteractingController)
            (if _returnvalue
              (bind _asynctaskproxy (Ability|Tasks|WaitGameplayTagAddToActor (Game|Player|GetPlayerPawn InteractingController) "(TagName=\"Gameplay.Action.Player.Emote.Dance\")")
                (:then
                  (Utilities|Time|ClearandInvalidateTimerbyHandle (Variables|State|GetPlayerDanceReactionTimerHandle)))
                (:Added
                  (Utilities|Time|SetTimerbyEvent (delegate Custom|Event%20Player%20Dance%20For%20Required%20Duration) (Variables|Responses|GetRequiredPlayerDanceDurationToReact)))))
            (else
              (CallFunction|ChangeEmoteScreenRevealed false)
              (Variables|State|SetLastInteractingPlayerController 0)
              (if _returnvalue
                (Utilities|Time|ClearandInvalidateTimerbyHandle (Variables|State|GetPlayerDanceReactionTimerHandle))
                (Utilities|IsValid _asynctaskproxy
                  (:Is%20Valid
                    (Ability|Async|EndAction _asynctaskproxy))))))
          (else
            (Variables|State|SetLastDialogMarkerInteractionState CurrentInteractionState)))))))

(event Custom|Event%20On%20Receive%20Conversation%20Message
  (params (LocalPlayerController :type Fort%20Player%20Controller%20Object%20Reference) (Message :type Client%20Conversation%20Message%20Payload%20Structure%20%28by%20ref%29))
  ;[1 1 1] Event handling - Check for metadata driven Emote Trigger when we receive a new conversation message from the local player
  (CallFunction|HandleReceiveConversationMessage Message))

(event EventOnDamagePlayEffects
  (params (Damage :type Float%20%28single-precision%29) (DamageTags :type Gameplay%20Tag%20Container%20Structure%20%28by%20ref%29) (Momentum :type Vector :default "0, 0, 0") (HitInfo :type Hit%20Result%20Structure%20%28by%20ref%29) (InstigatedBy :type Fort%20Pawn%20Object%20Reference) (DamageCauser :type Actor%20Object%20Reference) (EffectContext :type Gameplay%20Effect%20Context%20Handle%20Structure))
  ;[1 1 1] Event handling - Check for Emote Trigger when we receive damage. Note OnTakeDamage does not get hit. Reacting via effects route. Only provides pawn which is less desirable than controller but should be fine for now.
  (Default|HandleReceiveDamage InstigatedBy))

(event EventTick
  (params (DeltaSeconds :type Float%20%28single-precision%29))
  (|CallParentFunction DeltaSeconds) ; This node is disabled and will not be called.\nDrag off pins to build functionality.
)

(event Custom|TriggerVendEffects
  (bind _p_loot_effect_vending (Variables|Default|GetP_Loot_Effect_Vending))
  (Networking|SwitchHasAuthority
    (:Authority
      (CallFunction|VendSucceedSFX))
    (:Remote
      (Utilities|FlowControl|Sequence
        (:then_0
          _)
        (:then_1
          (Utilities|IsValid _p_loot_effect_vending
            (:Is%20Valid
              (Components|Activation|SetActive _p_loot_effect_vending true true)))))))
  (bind (_direction _wobbleovertime _flash) (Timeline|WobbleTimeline PlayFromStart)
    (:Update
      (Utilities|FlowControl|Sequence
        (:then_0
          (if (Variables|Materials|GetMaterialEffectsEnabled)
            (Rendering|Material|SetScalarParameterValue (Variables|Materials|GetMonitor%20Mat) (Variables|Materials|GetMonitorMaterialParamName) _flash :!class /Script/Engine.MaterialInstanceDynamic)))
        (:then_1
          (bind _returnvalue (BuildingActor|PhysicsSimulationSupport|GetPhysicsObjectComponent self))
          (if (not (PhysicsObject|IsAwake _returnvalue))
            (Transformation|SetRelativeRotation (Variables|Debug|GetStaticMeshComponent) (Math|Rotator|CombineRotators (Variables|State|GetInitialRelativeRotation) (Math|Rotator|Lerp(Rotator) "0, 0, 0" "0.500000,0.000000,0.000000" _wobbleovertime true)) false false)))))
    (:Finished
      (if (Variables|Materials|GetMaterialEffectsEnabled)
        (Rendering|Material|SetScalarParameterValue (Variables|Materials|GetMonitor%20Mat) (Variables|Materials|GetMonitorMaterialParamName) 0.0 :!class /Script/Engine.MaterialInstanceDynamic)))))

(event Custom|Vend%20Succeed%20SFX
  (bind _p_loot_effect_vending (Variables|Default|GetP_Loot_Effect_Vending))
  (Networking|CanExecuteCosmeticEvents
    (:True
      (Audio|PlaySoundatLocation (Variables|Behavior|GetVend%20Succeeded%20Sound) (Transformation|GetActorLocation))
      (Utilities|IsValid _p_loot_effect_vending
        (:Is%20Valid
          (Components|Activation|SetActive _p_loot_effect_vending true true))))))

(event Custom|Reset%20A%20Defaults
  (Default|ResetADefaults))

(event Custom|ChangeEmote
  (params (DesiredStyle :type Vending%20Machine%20Material%20Style%20Structure))
  (bind (_transitiontimeline_14_4015721f44f75fde5a0980929bfdfe02 _useblink_21_03c1119f421027edd25feca30481bb78 _uselaugh_23_7f967acc46f5050fcaede8ac8c6b6fc2 _ontriggeredsound_30_f893df0d4e55a6e55a6de9954603d3b1) (Utilities|Struct|BreakVendingMachineMaterialStyle DesiredStyle))
  (if (Health|HasHealthLeft)
    (Utilities|FlowControl|Sequence
      (:then_0
        ;[1 1 1] Store previous style as A, desired as B
        (Variables|Materials|SetMaterialStyleA (Variables|Materials|GetMaterialStyleB))
        (Variables|Materials|SetMaterialStyleB DesiredStyle)
        (CallFunction|SetMaterialTextureParameters)
        (Variables|Materials|SetDistortionMultiplier 1.0)
        (switch name _transitiontimeline_14_4015721f44f75fde5a0980929bfdfe02
          (:NoDistort
            (Variables|Materials|SetDistortionMultiplier 0.0))
          (:AToB
            _)
          (:BToA
            (bind (_direction _distortion _emote_swap _background_change _cloud_and_rainbow) (Timeline|From_B_to_A PlayFromStart)
              (:Update
                (CallFunction|UpdateMaterialTransition _distortion _emote_swap _background_change _cloud_and_rainbow)))))
        (bind (_direction_2 _distortion_2 _emote_swap_2 _background_change_2 _clour_and_rainbow) (Timeline|From_A_to_B PlayFromStart)
          (:Update
            (CallFunction|UpdateMaterialTransition _distortion_2 _emote_swap_2 _background_change_2 _clour_and_rainbow))))
      (:then_1
        ;[1 1 1] If there is an associated audio cue to trigger when this emote is triggered play it now. If was same as previous, do not play, to avoid spam. If screen not revealed, do not play.
        (if (ScalableFloat|GetValueasBool (Variables|Responses|GetTriggerAudioOnEmote))
          (if (Utilities|NotEqual(Object) (Variables|State|GetLastPlayedEmoteAudio) _ontriggeredsound_30_f893df0d4e55a6e55a6de9954603d3b1)
            (Variables|State|SetLastPlayedEmoteAudio _ontriggeredsound_30_f893df0d4e55a6e55a6de9954603d3b1)
            (Utilities|IsValid _ontriggeredsound_30_f893df0d4e55a6e55a6de9954603d3b1
              (:Is%20Valid
                (bind _current_reaction_audio (Variables|Components|GetCurrent%20Reaction%20Audio)
                  (:then
                    (Audio|Components|Audio|FadeOut _current_reaction_audio 0.5 0.0 "Linear" :!class /Script/Engine.AudioComponent))
                  (:else
                    _))
                (if (Variables|State|GetEmoteScreenRevealed)
                  (bind _returnvalue (Audio|SpawnSoundAttached (Variables|State|GetLastPlayedEmoteAudio) (Variables|Default|GetInteractSoundComponent)))
                  (Variables|Components|SetCurrent%20Reaction%20Audio _returnvalue))))))))))

(event Custom|ChangeEmoteScreenRevealed
  (params (Revealed :type Boolean))
  (bind _monitor_mat (Variables|Materials|GetMonitor%20Mat))
  (Variables|State|SetEmoteScreenRevealed Revealed)
  (if (Health|HasHealthLeft)
    (Utilities|FlowControl|Sequence
      (:then_0
        ;[1 1 1] Hook me in if we want a 0-1 value that triggers when crossing non-interactive -> interactive threshold
        ; Sympathetic to player thrashing over trigger line (will reverse from current point in transition).
        ; Alternative would be to Allow transition to fully finish, and then trigger the reverse once complete which could be done pretty simply by storing the last requested state, whether a transition is in progress, then hooking up the 'finished' output to a new trigger if no longer in correct state
        (if Revealed
          (bind (_direction _emotevisibility _transientdistortion) (Timeline|RevealEmoteScreen Play)
            (:Update
              (Rendering|Material|SetScalarParameterValue _monitor_mat "bIsLlama" _emotevisibility :!class /Script/Engine.MaterialInstanceDynamic)
              (Rendering|Material|SetScalarParameterValue _monitor_mat "DistortionAdd" _transientdistortion :!class /Script/Engine.MaterialInstanceDynamic)))
          (else
            (Timeline|RevealEmoteScreen Reverse))))
      (:then_1
        ;[1 1 1] If enabled, play a transition in / out sound
        (if (ScalableFloat|GetValueasBool (Variables|Responses|GetTriggerAudioOnScreenTransition))
          (if Revealed
            (Utilities|IsValid (Variables|Responses|GetOnRevealScreenSound)
              (:Is%20Valid
                (bind _onrevealaudio (Variables|Default|GetOnRevealAudio)
                  (:then
                    (Audio|Components|Audio|FadeOut _onrevealaudio 0.5 0.0 "Linear" :!class /Script/Engine.AudioComponent))
                  (:else
                    _))
                (bind _returnvalue (Audio|SpawnSoundatLocation (Variables|Responses|GetOnRevealScreenSound) (Transformation|GetWorldLocation (Variables|Default|GetInteractSoundComponent))))
                (Variables|Default|SetOnRevealAudio _returnvalue)
                (bind _onhideaudio (Variables|Default|GetOnHideAudio)
                  (:then
                    (Audio|Components|Audio|FadeOut _onhideaudio 0.5 0.0 "Linear" :!class /Script/Engine.AudioComponent))
                  (:else))))
            (else
              (Utilities|IsValid (Variables|Responses|GetOnHideScreenSound)
                (:Is%20Valid
                  (bind _onhideaudio_2 (Variables|Default|GetOnHideAudio)
                    (:then
                      (Audio|Components|Audio|FadeOut _onhideaudio_2 0.5 0.0 "Linear" :!class /Script/Engine.AudioComponent))
                    (:else
                      _))
                  (bind _returnvalue_2 (Audio|SpawnSoundatLocation (Variables|Responses|GetOnHideScreenSound) (Transformation|GetWorldLocation (Variables|Default|GetInteractSoundComponent))))
                  (Variables|Default|SetOnHideAudio _returnvalue_2)
                  (CallFunction|StopReactionAudio 1.0)
                  (bind _onrevealaudio_2 (Variables|Default|GetOnRevealAudio)
                    (:then
                      (Audio|Components|Audio|FadeOut _onrevealaudio_2 0.5 0.0 "Linear" :!class /Script/Engine.AudioComponent))
                    (:else)))))))))))

(event Custom|StopReactionAudio
  (params (FadeOutDuration :type Float%20%28double-precision%29))
  ;[1 1 1] If enabled, play a transition in / out sound
  (bind _current_reaction_audio (Variables|Components|GetCurrent%20Reaction%20Audio)
    (:then
      (Audio|Components|Audio|FadeOut _current_reaction_audio FadeOutDuration 0.0 "Linear" :!class /Script/Engine.AudioComponent))
    (:else)))

(event EventDestroyed
  ;[1 1 1] If enabled, play a transition in / out sound
  (CallFunction|StopReactionAudio 0.25)
  (bind _onrevealaudio (Variables|Default|GetOnRevealAudio)
    (:then
      (Audio|Components|Audio|FadeOut _onrevealaudio 0.5 0.0 "Linear" :!class /Script/Engine.AudioComponent))
    (:else
      _))
  (bind _onhideaudio (Variables|Default|GetOnHideAudio)
    (:then
      (Audio|Components|Audio|FadeOut _onhideaudio 0.5 0.0 "Linear" :!class /Script/Engine.AudioComponent))
    (:else)))

(fn BlueprintOnLocalInteract
  (params (InteractingPawn :type Fort%20Player%20Pawn%20Object%20Reference) (InteractionBeingAttempted :type EInteractionBeingAttempted%20Enum :default "FirstInteraction"))
  (Utilities|FlowControl|Sequence
    (:then_0
      (if (Pawn|IsLocallyControlled InteractingPawn)
        (CallFunction|HandleLocalPlayerInteract)))
    (:then_1
      (return false))))

(fn EMPSetItemAndSpawnCount
  (Variables|EMP|Vars|SetEMP_ItemToSpawn (Variables|EMP|Item|GetItem_SpawnWhenEMPd))
  (Variables|EMP|Vars|SetEMP_ItemCountToSpawn (Math|Float|Truncate (ScalableFloat|GetValueatLevel (Variables|EMP|Balance|GetRow_ItemSpawnCount) (Math|Conversions|ToFloat(Integer) (Math|Integer|SelectInt (Math|Random|RandomInteger (Math|Float|Truncate (+ (ScalableFloat|GetValueatLevel (Variables|EMP|Balance|GetRow_MaxLevelChoiceForItemSpawnCount)) 1.0))) 0 (> (Math|Float|Truncate (+ (ScalableFloat|GetValueatLevel (Variables|EMP|Balance|GetRow_MaxLevelChoiceForItemSpawnCount)) 1.0)) 0)))))))

(fn GetDialogInfo
  (return (Variables|Responses|GetOnPlayerDanceMessage) (Variables|Responses|GetPlayerDanceMatStyleName) (Variables|Responses|GetSpecificPlayerDanceMatStyleNames) (Variables|Responses|GetSpecificOnPlayerDanceMessages) (Variables|Responses|GetDamageResponses_Light) (Variables|Responses|GetDamageResponses_Med) (Variables|Responses|GetDamageResponses_Heavy)))

(fn HandleLocalPlayerInteract
  (Utilities|IsValid (Variables|Responses|GetOnPlayerInteractSound)
    (:Is%20Valid
      (Audio|Components|Audio|SetSound (Variables|Default|GetInteractSoundComponent) (Variables|Responses|GetOnPlayerInteractSound))
      (Audio|Components|Audio|Play (Variables|Default|GetInteractSoundComponent) 0.0))))

(fn HandleReceiveConversationMessage
  (params (ClientConversationMessagePayload :type Client%20Conversation%20Message%20Payload%20Structure%20%28by%20ref%29))
  (bind (_speakerid _participantdisplayname _text _metadataparameters) (Utilities|Struct|BreakClientConversationMessage (Utilities|Struct|BreakClientConversationMessagePayload ClientConversationMessagePayload)))
  (bind _array_element (Utilities|Array|ForEachLoopwithBreak _metadataparameters :!label _label)
    (:LoopBody
      (if (Utilities|String|EqualExactly(String) (Utilities|Struct|BreakConversationNodeParameterPair _array_element) "Style")
        (Class|FortMaterialStyleComponent|RequestStyleChange (Variables|Default|GetFortMaterialStyleComponent) (Utilities|Struct|MakeRequestStyleParams -1.0 (Variables|Materials|GetMonitor%20Mat) "StyleSetLookup" (Variables|Materials|GetMaterialStyleSet) (Utilities|String|StringToName (.Value (Utilities|Struct|BreakConversationNodeParameterPair _array_element)))) "ReplaceActive")
        (goto _label Break)))))

(fn HandleReceiveDamage
  (params (InstigatingPawn :type Fort%20Pawn%20Object%20Reference))
  (bind _damageresponses_med (Variables|Responses|GetDamageResponses_Med))
  (bind (_outitem _outindex) (Utilities|Array|RandomArrayItem _damageresponses_med))
  (bind _damageresponses_light (Variables|Responses|GetDamageResponses_Light))
  (bind (_outitem_2 _outindex_2) (Utilities|Array|RandomArrayItem _damageresponses_light))
  (bind _damageresponses_heavy (Variables|Responses|GetDamageResponses_Heavy))
  (bind (_outitem_3 _outindex_3) (Utilities|Array|RandomArrayItem _damageresponses_heavy))
  (bind _chosendamageresponse_damagematerialstyle_5_da2e9bda41336285b6f136818c041fae (Utilities|Struct|BreakVendingDamageResponse (Variables|GetChosenDamageResponse_DamageMaterialStyle_5_DA2E9BDA41336285B6F136818C041FAE)))
  (bind (_chosendamageresponse_damagematerialstyle_5_da2e9bda41336285b6f136818c041fae_2 _chosendamageresponse_damagemessage_7_92e19af44ab84add702ce4bc7d00adf3) (Utilities|Struct|BreakVendingDamageResponse (Variables|GetChosenDamageResponse_DamageMaterialStyle_5_DA2E9BDA41336285B6F136818C041FAE)))
  (bind (_chosendamageresponse_damagematerialstyle_5_da2e9bda41336285b6f136818c041fae_3 _chosendamageresponse_damagemessage_7_92e19af44ab84add702ce4bc7d00adf3_2) (Utilities|Struct|BreakVendingDamageResponse (Variables|GetChosenDamageResponse_DamageMaterialStyle_5_DA2E9BDA41336285B6F136818C041FAE)))
  (bind (_chosendamageresponse_damagematerialstyle_5_da2e9bda41336285b6f136818c041fae_4 _chosendamageresponse_damagemessage_7_92e19af44ab84add702ce4bc7d00adf3_3) (Utilities|Struct|BreakVendingDamageResponse (Variables|GetChosenDamageResponse_DamageMaterialStyle_5_DA2E9BDA41336285B6F136818C041FAE)))
  (bind (_chosendamageresponse_damagematerialstyle_5_da2e9bda41336285b6f136818c041fae_5 _chosendamageresponse_damagemessage_7_92e19af44ab84add702ce4bc7d00adf3_4) (Utilities|Struct|BreakVendingDamageResponse (Variables|GetChosenDamageResponse_DamageMaterialStyle_5_DA2E9BDA41336285B6F136818C041FAE)))
  (if (or (ScalableFloat|GetValueasBool (Variables|Responses|GetTriggerSpeechOnTakeDamage)) (ScalableFloat|GetValueasBool (Variables|Responses|GetTriggerEmoteOnTakeDamage)))
    (Utilities|IsValid InstigatingPawn
      (:Is%20Valid
        (if (Pawn|IsLocallyControlled InstigatingPawn)
          ;[1 1 1] Trigger in world speech bubble based message based on which health threshold we are in
          ; Only trigger a message if a) its been long enough since the last OR b) we cross a health threshold c) we are in close enough range
          (Utilities|FlowControl|Sequence
            (:then_0
              (LocalVariables|SetChosenDamageResponse (Utilities|Struct|MakeVendingDamageResponse "Shock"))
              (LocalVariables|SetChosenDamageResponseThreshold -1)
              (if (> (Health|GetHealthPercent) 0.66)
                (if (Utilities|Array|IsValidIndex _damageresponses_light _outindex_2)
                  (LocalVariables|SetChosenDamageResponse _outitem_2)
                  (LocalVariables|SetChosenDamageResponseThreshold 0))
                (elif (> (Health|GetHealthPercent) 0.33)
                  (if (Utilities|Array|IsValidIndex _damageresponses_med _outindex)
                    (LocalVariables|SetChosenDamageResponse _outitem)
                    (LocalVariables|SetChosenDamageResponseThreshold 1))
                  (elif (Utilities|Array|IsValidIndex _damageresponses_heavy _outindex_3)
                    (LocalVariables|SetChosenDamageResponse _outitem_3)
                    (LocalVariables|SetChosenDamageResponseThreshold 2)))))
            (:then_1
              (if (ScalableFloat|GetValueasBool (Variables|Responses|GetTriggerSpeechOnTakeDamage))
                (if (and (or (and (Utilities|Text|NotEqual,CaseInsensitive(Text) (Variables|State|GetLastDisplayedTakeDamageMessage) _chosendamageresponse_damagemessage_7_92e19af44ab84add702ce4bc7d00adf3) (!= (Variables|State|GetLastDamageResponseThreshold) (LocalVariables|GetChosenDamageResponseThreshold))) (> (- (Fortnite|Sync|Time|GetServerWorldTimeSeconds) (Variables|State|GetLastDamageMessageTime)) (Variables|Responses|GetDamageMessageCooldown))) (or (Utilities|Enum|Equal(Enum) (Variables|State|GetLastDialogMarkerInteractionState) "Conversation") (Utilities|Enum|Equal(Enum) (Variables|State|GetLastDialogMarkerInteractionState) "InteractionRange")))
                  (if (not (Utilities|Text|TextIsEmpty _chosendamageresponse_damagemessage_7_92e19af44ab84add702ce4bc7d00adf3_2))
                    (InteractionUI|ShowMessage (Variables|Components|GetNonPlayerConversationComponent) _chosendamageresponse_damagemessage_7_92e19af44ab84add702ce4bc7d00adf3_3)
                    (Variables|State|SetLastDisplayedTakeDamageMessage _chosendamageresponse_damagemessage_7_92e19af44ab84add702ce4bc7d00adf3_4)
                    (Variables|State|SetLastDamageMessageTime (Fortnite|Sync|Time|GetServerWorldTimeSeconds))))))
            (:then_2
              (if (ScalableFloat|GetValueasBool (Variables|Responses|GetTriggerEmoteOnTakeDamage))
                (Class|FortMaterialStyleComponent|RequestStyleChange (Variables|Default|GetFortMaterialStyleComponent) (Utilities|Struct|MakeRequestStyleParams -1.0 (Variables|Materials|GetMonitor%20Mat) "StyleSetLookup" (Variables|Materials|GetMaterialStyleSet) _chosendamageresponse_damagematerialstyle_5_da2e9bda41336285b6f136818c041fae) "ReplaceActive")))
            (:then_3
              (Variables|State|SetLastDamageResponseThreshold (LocalVariables|GetChosenDamageResponseThreshold)))))))))

(fn PlayServiceUsedEffects
  (params (PlayerController :type Fort%20Player%20Controller%20Object%20Reference))
  (CallFunction|TriggerVendEffects))

(fn ResetADefaults
  (bind _materialstylea (Variables|Materials|GetMaterialStyleA))
  (bind (_texture_3_114ba7ba459752bb759266a2c7dfae22 _bgcolor1_10_158f7a5f4e78b4bf2649e5a046672ed2 _bgcolor2_11_fe5c46e543bb4ca35915b794884e9d29) (Utilities|Struct|BreakVendingMachineMaterialStyle _materialstylea))
  (bind _monitor_mat (Variables|Materials|GetMonitor%20Mat))
  (Rendering|Material|SetScalarParameterValue _monitor_mat "EmoteSwap" 0.0 :!class /Script/Engine.MaterialInstanceDynamic)
  (Rendering|Material|SetScalarParameterValue _monitor_mat "Distort" 0.0 :!class /Script/Engine.MaterialInstanceDynamic)
  (Rendering|Material|SetTextureParameterValue (Variables|Materials|GetMonitor%20Mat) "Emote_B" _texture_3_114ba7ba459752bb759266a2c7dfae22)
  (Rendering|Material|SetVectorParameterValue (Variables|Materials|GetMonitor%20Mat) "BG_1_RGBA" _bgcolor1_10_158f7a5f4e78b4bf2649e5a046672ed2 :!class /Script/Engine.MaterialInstanceDynamic)
  (Rendering|Material|SetVectorParameterValue (Variables|Materials|GetMonitor%20Mat) "BG_2_RGBA" _bgcolor2_11_fe5c46e543bb4ca35915b794884e9d29 :!class /Script/Engine.MaterialInstanceDynamic))

(fn SetMaterialTextureParameters
  (Rendering|Material|SetTextureParameterValue (Variables|Materials|GetMonitor%20Mat) "Emote_A" (Utilities|Struct|BreakVendingMachineMaterialStyle (Variables|Materials|GetMaterialStyleA)))
  (Rendering|Material|SetTextureParameterValue (Variables|Materials|GetMonitor%20Mat) "Emote_B" (Utilities|Struct|BreakVendingMachineMaterialStyle (Variables|Materials|GetMaterialStyleB))))

(fn TriggerChangeMaterialStyle
  (params (MaterialStyleData :type Vending%20Machine%20Material%20Style%20Structure))
  (CallFunction|ChangeEmote MaterialStyleData))

(fn TriggerDanceReaction
  (bind _lastinteractingplayercontroller (Variables|State|GetLastInteractingPlayerController))
  (bind _returnvalue (Emote|GetLastEmotePlayed _lastinteractingplayercontroller))
  (bind _lastinteractingplayercontroller_2 (Variables|State|GetLastInteractingPlayerController))
  (bind _returnvalue_2 (Emote|GetLastEmotePlayed _lastinteractingplayercontroller_2))
  (bind _output (Utilities|ResolveSoftReference _returnvalue_2))
  (bind _specificplayerdancematstylenames (Variables|Responses|GetSpecificPlayerDanceMatStyleNames))
  (bind (_value _returnvalue_3) (Utilities|Map|Find _specificplayerdancematstylenames _output))
  (bind _output_2 (Utilities|ResolveSoftReference _returnvalue))
  (bind _specificonplayerdancemessages (Variables|Responses|GetSpecificOnPlayerDanceMessages))
  (bind (_value_2 _returnvalue_4) (Utilities|Map|Find _specificonplayerdancemessages _output_2))
  (LocalVariables|SetMatStyleName (Variables|Responses|GetPlayerDanceMatStyleName))
  (Utilities|FlowControl|Sequence
    (:then_0
      (Utilities|IsValid _lastinteractingplayercontroller_2
        (:Is%20Valid
          (Utilities|IsValid _returnvalue_2
            (:Is%20Valid
              (if _returnvalue_3
                (if (Utilities|Name|NotEqual(Name) _value "None")
                  (LocalVariables|SetMatStyleName _value))))))))
    (:then_1
      (Class|FortMaterialStyleComponent|RequestStyleChange (Variables|Default|GetFortMaterialStyleComponent) (Utilities|Struct|MakeRequestStyleParams -1.0 (Variables|Materials|GetMonitor%20Mat) "StyleSetLookup" (Variables|Materials|GetMaterialStyleSet) (LocalVariables|GetMatStyleName)) "ReplaceActive")
      (if (ScalableFloat|GetValueasBool (Variables|Responses|GetTriggerSpeechOnLocalPlayerDance))
        (LocalVariables|SetMessage (Variables|Responses|GetOnPlayerDanceMessage))
        (Utilities|FlowControl|Sequence
          (:then_0
            (Utilities|IsValid _lastinteractingplayercontroller
              (:Is%20Valid
                (Utilities|IsValid _returnvalue
                  (:Is%20Valid
                    (if _returnvalue_4
                      (if (not (Utilities|Text|TextIsEmpty _value_2))
                        (LocalVariables|SetMessage _value_2))))))))
          (:then_1
            (InteractionUI|ShowMessage (Variables|Components|GetNonPlayerConversationComponent) (LocalVariables|GetMessage))))))))

(fn UpdateMaterialTransition
  (params (Distortion :type Float%20%28double-precision%29) (EmoteSwap :type Float%20%28double-precision%29) (BackgroundChange :type Float%20%28double-precision%29) (Cloud%20and%20Rainbow :type Float%20%28double-precision%29))
  (bind _materialstylea (Variables|Materials|GetMaterialStyleA))
  (bind (_bgcolor1_10_158f7a5f4e78b4bf2649e5a046672ed2 _bgcolor2_11_fe5c46e543bb4ca35915b794884e9d29 _transitiontimeline_14_4015721f44f75fde5a0980929bfdfe02 _cloudandrainbow_17_81d2450d40a82032545405a9a31688dd _useblink_21_03c1119f421027edd25feca30481bb78 _uselaugh_23_7f967acc46f5050fcaede8ac8c6b6fc2) (Utilities|Struct|BreakVendingMachineMaterialStyle _materialstylea))
  (bind _materialstyleb (Variables|Materials|GetMaterialStyleB))
  (bind (_bgcolor1_10_158f7a5f4e78b4bf2649e5a046672ed2_2 _bgcolor2_11_fe5c46e543bb4ca35915b794884e9d29_2 _transitiontimeline_14_4015721f44f75fde5a0980929bfdfe02_2 _cloudandrainbow_17_81d2450d40a82032545405a9a31688dd_2 _useblink_21_03c1119f421027edd25feca30481bb78_2 _uselaugh_23_7f967acc46f5050fcaede8ac8c6b6fc2_2) (Utilities|Struct|BreakVendingMachineMaterialStyle _materialstyleb))
  (Rendering|Material|SetScalarParameterValue (Variables|Materials|GetMonitor%20Mat) "Distort" (* (Variables|Materials|GetDistortionMultiplier) Distortion) :!class /Script/Engine.MaterialInstanceDynamic)
  (Rendering|Material|SetScalarParameterValue (Variables|Materials|GetMonitor%20Mat) "EmoteSwap" EmoteSwap :!class /Script/Engine.MaterialInstanceDynamic)
  (Rendering|Material|SetVectorParameterValue (Variables|Materials|GetMonitor%20Mat) "BG_1_RGBA" (Math|Color|Lerp(LinearColor) _bgcolor1_10_158f7a5f4e78b4bf2649e5a046672ed2 _bgcolor1_10_158f7a5f4e78b4bf2649e5a046672ed2_2 BackgroundChange) :!class /Script/Engine.MaterialInstanceDynamic)
  (Rendering|Material|SetVectorParameterValue (Variables|Materials|GetMonitor%20Mat) "BG_2_RGBA" (Math|Color|Lerp(LinearColor) _bgcolor2_11_fe5c46e543bb4ca35915b794884e9d29 _bgcolor2_11_fe5c46e543bb4ca35915b794884e9d29_2 BackgroundChange) :!class /Script/Engine.MaterialInstanceDynamic)
  (Rendering|Material|SetVectorParameterValue (Variables|Materials|GetMonitor%20Mat) "Cloud and Rainbow" (Math|Color|Lerp(LinearColor) _cloudandrainbow_17_81d2450d40a82032545405a9a31688dd _cloudandrainbow_17_81d2450d40a82032545405a9a31688dd_2 Cloud%20and%20Rainbow) :!class /Script/Engine.MaterialInstanceDynamic)
  (Rendering|Material|SetScalarParameterValue (Variables|Materials|GetMonitor%20Mat) "bUseBlink" (Math|Float|Lerp _useblink_21_03c1119f421027edd25feca30481bb78 _useblink_21_03c1119f421027edd25feca30481bb78_2 Cloud%20and%20Rainbow) :!class /Script/Engine.MaterialInstanceDynamic)
  (Rendering|Material|SetScalarParameterValue (Variables|Materials|GetMonitor%20Mat) "bUseLaugh" (Math|Float|Lerp _uselaugh_23_7f967acc46f5050fcaede8ac8c6b6fc2 _uselaugh_23_7f967acc46f5050fcaede8ac8c6b6fc2_2 Cloud%20and%20Rainbow) :!class /Script/Engine.MaterialInstanceDynamic))

(fn ConstructionScript
  (|CallParentFunction))
